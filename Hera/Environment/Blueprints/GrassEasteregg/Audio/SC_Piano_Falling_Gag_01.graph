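# Sound cue for a falling piano gag, apparently part of a grass easter egg in the Hera environment.
# A mixer plays the falling whistle or descent sound immediately and layers a louder crash that starts after a fixed two-second delay.
# The cue uses the Athena ambient sound class with boosted volume and its own attenuation override, including occlusion with a low-pass filter on a custom trace channel.
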SoundCue /Hera/Environment/Blueprints/GrassEasteregg/Audio/SC_Piano_Falling_Gag_01
root=Mixer_0
nodes [
 {id=Delay_0, children [WavePlayer_1], properties {delayMax=2, delayMin=2}}
 {id=Mixer_0, children [WavePlayer_0,Delay_0], properties {inputVolume [1,2]}}
 {id=WavePlayer_0, properties {soundWaveAssetPtr=&/BR_SFX_Piano_Gag_Falling_01}}
 {id=WavePlayer_1, properties {soundWaveAssetPtr=&/BR_SFX_Piano_Gag_Crash_01}}
]
properties {
 bOverrideAttenuation=true
 maxDistance=3600
 soundClassObject=&Fort_Ambient_Athena
 volumeMultiplier=3
 attenuationOverrides {bEnableOcclusion=true, occlusionLowPassFilterFrequency=10000, occlusionTraceChannel=ECC_GameTraceChannel11, attenuationShapeExtents {x=0}}
}
