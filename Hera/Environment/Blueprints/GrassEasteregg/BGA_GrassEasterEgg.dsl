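; Invulnerable grass easter egg actor that drops a piano from a portal onto players lingering in its box.
; Server overlap events track pawns and run a six-second check on one, with logic to retry another.
; A successful check fires a one-shot Timeline driving portal effects, audio and the mesh's fall, spin and scale.
; On landing it plays crash effects, destroys listed building actors and deals lethal environment damage to pawns inside.
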(event EventBeginPlay
  (Variables|BuildingActor|SetbIsInvulnerable true)
  (Variables|Default|SetMeshEndLocation (Transformation|GetWorldLocation (Variables|Default|GetBillboard)))
  (Variables|Default|SetMeshStationLocation (Transformation|GetWorldLocation (Variables|Default|GetPortalBillboard))))

(event OnComponentBeginOverlap(Box)
  (params (OverlappedComponent :type Primitive%20Component%20Object%20Reference) (OtherActor :type Actor%20Object%20Reference) (OtherComp :type Primitive%20Component%20Object%20Reference) (OtherBodyIndex :type Integer) (bFromSweep :type Boolean) (SweepResult :type Hit%20Result%20Structure%20%28by%20ref%29))
  (bind _asfort_player_pawn_athena (Utilities|Casting|CastToFortPlayerPawnAthena OtherActor))
  (Networking|SwitchHasAuthority
    (:Authority
      (Utilities|Array|AddUnique (Variables|Default|GetAllOverlappingPawns) _asfort_player_pawn_athena)
      (Utilities|FlowControl|DoOnce false
        (:Completed
          (Variables|Default|SetInitPawnRef _asfort_player_pawn_athena)
          (bind _returnvalue (Utilities|Time|SetTimerbyEvent (delegate Custom|CheckOverlap) 6.0))
          (Variables|Default|SetTimerHandle _returnvalue))))))

(event Custom|ResetCheckOnce
  (Utilities|FlowControl|DoOnce false
    (:Completed
      (Variables|Default|SetInitPawnRef (stale OnComponentBeginOverlap(Box) _asfort_player_pawn_athena))
      (bind _returnvalue (Utilities|Time|SetTimerbyEvent (delegate Custom|CheckOverlap) 6.0))
      (Variables|Default|SetTimerHandle _returnvalue))))

(event Custom|RetriggerCheck
  (params (Pawn :type Fort%20Player%20Pawn%20Athena%20Object%20Reference))
  (Variables|Default|SetInitPawnRef Pawn)
  (bind _returnvalue (Utilities|Time|SetTimerbyEvent (delegate Custom|CheckOverlap) 6.0))
  (Variables|Default|SetTimerHandle _returnvalue))

(event Custom|CheckOverlap
  (bind _initpawnref (Variables|Default|GetInitPawnRef))
  (Utilities|IsValid _initpawnref
    (:Is%20Valid
      (if (Utilities|Array|ContainsItem (Collision|GetOverlappingActors (Variables|Default|GetBox) "/Script/FortniteGame.FortPlayerPawnAthena" :!class /Script/Engine.PrimitiveComponent) _initpawnref)
        (Utilities|FlowControl|DoOnce false
          (:Completed
            (CallFunction|ExecuteSquashEvent)))
        (else
          (CallFunction|RetryAnotherPlayer))))))

(event OnComponentEndOverlap(Box)
  (params (OverlappedComponent :type Primitive%20Component%20Object%20Reference) (OtherActor :type Actor%20Object%20Reference) (OtherComp :type Primitive%20Component%20Object%20Reference) (OtherBodyIndex :type Integer))
  (bind _asfort_player_pawn_athena (Utilities|Casting|CastToFortPlayerPawnAthena OtherActor))
  (Networking|SwitchHasAuthority
    (:Authority
      (Utilities|Array|RemoveItem (Variables|Default|GetAllOverlappingPawns) _asfort_player_pawn_athena)
      (if (== _asfort_player_pawn_athena (Variables|Default|GetInitPawnRef))
        (Utilities|Time|ClearandInvalidateTimerbyHandle (Variables|Default|GetTimerHandle))
        (CallFunction|ResetCheckOnce)))))

(event Custom|RetryAnotherPlayer
  (bind _alloverlappingpawns (Variables|Default|GetAllOverlappingPawns))
  (bind _output (Utilities|Array|Get(acopy) _alloverlappingpawns))
  (if (> (Utilities|Array|Length (Collision|GetOverlappingActors (Variables|Default|GetBox) "/Script/FortniteGame.FortPlayerPawnAthena" :!class /Script/Engine.PrimitiveComponent)) 0)
    (Utilities|IsValid _output
      (:Is%20Valid
        (CallFunction|RetriggerCheck _output)))))

(event Custom|ExecuteSquashEvent
  (Utilities|FlowControl|DoOnce false
    (:Completed
      (bind (_direction _lerp _scale _rotation) (Timeline|Animation PlayFromStart)
        (:Update
          (Transformation|SetWorldLocationAndRotation (Variables|Default|GetScene) (Math|Vector|Lerp(Vector) (Variables|Default|GetMeshStationLocation) (- (Variables|Default|GetMeshEndLocation) "0.000000,0.000000,100.000000") _lerp) (Math|Rotator|Lerp(Rotator) "0, 0, 0" "400.000000,200.000000,1200.000000" _rotation true) false false)
          (Transformation|SetWorldScale3D (Variables|Default|GetAnimatingMesh) (Math|Conversions|ToVector(Float) (Math|Float|Lerp 0.0 1.0 _scale))))
        (:Finished
          (Components|Activation|Activate (Variables|Default|GetCrashEffects) false)
          (Rendering|SetVisibility (Variables|Default|GetAnimatingMesh) false false)
          (Audio|Components|Audio|FadeOut (Variables|Default|GetWarningAudio) 1.0 0.0 "Linear" :!class /Script/Engine.AudioComponent)
          (Audio|Components|Audio|Play (Variables|Default|GetCrashAudio) 0.0)
          (Networking|SwitchHasAuthority
            (:Authority
              (for _array_element (Variables|Default|GetBuildingActorToDestroy)
                (Utilities|IsValid _array_element
                  (:Is%20Valid
                    (Actor|DestroyActor _array_element))))
              (CallFunction|DownPlayers))))
        (:PortalOpen
          (Components|Activation|Activate (Variables|Default|GetPortalEffect) false)
          (Audio|Components|Audio|Play (Variables|Default|GetPortalOpenAudio) 0.0)
          (CallFunction|PianoFalling)
          (Rendering|SetVisibility (Variables|Default|GetAnimatingMesh) true false)
          (Audio|Components|Audio|Play (Variables|Default|GetWarningAudio) 0.0))))))

(event Custom|PianoFalling
  ;[1 1 1] Piano falling sound logic
  (Utilities|FlowControl|Delay (Variables|Default|GetDelaytime))
  (Audio|Components|Audio|Play (Variables|Default|GetPianoFallingAudio) 0.0))

(event Custom|DownPlayers
  (Utilities|FlowControl|DoOnce false
    (:Completed
      (Networking|SwitchHasAuthority
        (:Authority
          (for _array_element (Collision|GetOverlappingActors (Variables|Default|GetBox) "/Script/FortniteGame.FortPlayerPawnAthena" :!class /Script/Engine.PrimitiveComponent)
            (Utilities|IsValid _array_element
              (:Is%20Valid
                (Environment|ApplyGlobalEnvironmentDamagetoActor _array_element 9999.0)))))))))

(fn ConstructionScript)
