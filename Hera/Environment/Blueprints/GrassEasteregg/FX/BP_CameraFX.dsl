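; BP_CameraFX is a Blueprint for the grass easter egg that shows a temporary paint-can lens effect on the player's camera.
; On BeginPlay it caches player 0's camera manager, adds the BP_PaintCan_LensFX lens effect and stores the returned instance.
; A sequence branch removes that effect after a two-second delay, and EndPlay also removes it.
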(event EventBeginPlay
  (Utilities|FlowControl|Sequence
    (:then_0
      (Variables|Default|SetPlayerCameraManager (Game|GetPlayerCameraManager 0))
      (bind _returnvalue (Camera|AddGenericCameraLensEffect (Variables|Default|GetPlayerCameraManager) "/Hera/Environment/Blueprints/GrassEasteregg/FX/BP_PaintCan_LensFX.BP_PaintCan_LensFX_C"))
      (Variables|Default|Setcamlenseffect _returnvalue))
    (:then_1
      (Utilities|FlowControl|Delay 2.0)
      (Camera|RemoveGenericCameraLensEffect (Variables|Default|GetPlayerCameraManager) (Variables|Default|Getcamlenseffect)))))

(event EventEndPlay
  (params (EndPlayReason :type EEndPlayReason%20Enum :default "Destroyed"))
  (Camera|RemoveGenericCameraLensEffect (Variables|Default|GetPlayerCameraManager) (Variables|Default|Getcamlenseffect)))
