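# Default Lit material for Niagara mesh particles, shading the piano debris chunks in the GrassEasteregg effect.
# Vertex color green blends the grand piano diffuse, multiplied by its inverted red mask, with a darkened wooden-board texture, likely for broken interiors.
# A scalar parameter controls the wood's desaturation.
# The piano's normal and S-mask textures are also sampled, though the file shows no links to material outputs.
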Material /Hera/Environment/Blueprints/GrassEasteregg/FX/Materials/M_Piano_MeshParticle_Chunks_01
graph {
 properties {bAutomaticallySetUsageInEditor=false, bUsedWithNiagaraMeshParticles=true, usedShadingModels=Default Lit}
 nodes [
  n1 = TextureSample() {
   samplerType=SAMPLERTYPE_Masks
   texture=/Game/Environments/Helios/Props/Furniture/Furniture_GrandPianos_A/Textures/T_Furniture_GrandPiano_A_A_M
  }
  n2 = ComponentMask(None=@n1.RGB) {r=true}
  n3 = TextureSample() {texture=/Game/Environments/Asteria/Sets/LightCastle/Textures/T_LightCastle_WoodenCeilingBoards_01_D}
  n4 = ScalarParameter()
  LinearInterpolate(
   A=Multiply(
    A=OneMinus(None=@n2)
    B=TextureSample() {texture=/Game/Environments/Helios/Props/Furniture/Furniture_GrandPianos_A/Textures/T_Furniture_GrandPiano_A_A_D}.RGB
   )
   B=Multiply(A=Desaturation(None=@n3.RGB, Fraction=@n4)) {constB=0.25}
   Alpha=VertexColor().G
  )
  Saturate()
  TextureSample() {
   samplerType=SAMPLERTYPE_Normal
   texture=/Game/Environments/Helios/Props/Furniture/Furniture_GrandPianos_A/Textures/T_Furniture_GrandPiano_A_A_N
  }
  TextureSample() {
   samplerType=SAMPLERTYPE_Masks
   texture=/Game/Environments/Helios/Props/Furniture/Furniture_GrandPianos_A/Textures/T_Furniture_GrandPiano_A_A_S
  }
 ]
}
