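# M_Piano_MeshParticle_Keys is a Default Lit material for Niagara mesh particles, named for piano keys and stored with the grass easter egg effects.
# Its graph holds only a default-valued three-component color constant and a Particle Color node, with no connections between them or to material outputs shown.
# Automatic usage flagging in the editor is off, and Niagara mesh particle usage is enabled explicitly.
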Material /Hera/Environment/Blueprints/GrassEasteregg/FX/Materials/M_Piano_MeshParticle_Keys
graph {
 properties {bAutomaticallySetUsageInEditor=false, bUsedWithNiagaraMeshParticles=true, usedShadingModels=Default Lit}
 nodes [
  Constant3Vector()
  ParticleColor()
 ]
}
