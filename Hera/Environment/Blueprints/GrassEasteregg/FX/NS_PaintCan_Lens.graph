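# Niagara system for the grass easter egg's paint can effect, a one-shot purple paint splatter that, by its name, lands on the camera lens.
# A single sprite emitter bursts randomly sized particles onto the side surface of a thin cylinder, forming a ring offset along X.
# Curves pop the sprite size and fade alpha in quickly and out slowly, while curve-ramped gravity and drag let the droplets sag downward.
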NiagaraSystem /Hera/Environment/Blueprints/GrassEasteregg/FX/NS_PaintCan_Lens
dependencies {
 usedDataInterfaces [&NiagaraDataInterfaceCurve]
 usedDynamicInputs [&FNTimeOfDayScalar]
 usedModules [
  &SystemState
  &self:SystemSpawnScript.SystemScriptSource.SystemScriptGraph.NiagaraNodeAssignment_0.SetVariables_#1
 ]
}
emitters {Splash001 {
  emitterUpdateScript [
   &EmitterState()
   &SpawnBurst_Instantaneous(Spawn Count=24, Use Spawn Probability=false)
  ]
  particleSpawnScript [
   &InitializeParticle(
    Color Maximum=<rgba 0.1354 0 0.1264 1>
    Color Minimum=<rgba 0.2881 0 0.8073 1>
    Color Mode=::NewEnumerator2
    Lifetime Min=0.5
    Lifetime Mode=::NewEnumerator1
    Position Offset=<50 0 0>
    Sprite Rotation Mode=::NewEnumerator2
    Sprite Size Mode=::NewEnumerator3
    Uniform Sprite Size=&RandomRangeFloat(Minimum=6, Maximum=15)
    UsePositionOffset=true
    Write Material Random=true
   )
   &ShapeLocation(
    Cylinder Height=1
    Cylinder Radius=&RandomRangeFloat(Minimum=25, Maximum=50)
    MaskSpawnGroup=false
    Override Seed=false
    Rotation Mode=::NewEnumerator1
    Shape Primitive=::NewEnumerator1
    Surface Only=true
    Surface Only Band Thickness=0.1
    Use Endcaps In Surface Only Mode=false
    UseOwnerScale=true
   )
  ]
  particleUpdateScript [
   &ParticleState()
   &ScaleSpriteSize(
    Uniform Curve Index=$Particles.Age
    Uniform Curve Sprite Scale=DataInterface() {propertyValues {Curve {keys [
        <at 0 0.6467 RCIM_Cubic>
        {arriveTangent=5.4479, interpMode=RCIM_Cubic, leaveTangent=-0.3343, tangentMode=RCTM_Break, time=0.0553, value=1}
        <at 1 0.9046 RCIM_Cubic>
       ]}}}
   ) {module=ScaleSpriteSize001}
   &ScaleSpriteSize()
   &ScaleColor(
    Scale Alpha=&FloatFromCurve(FloatCurve=DataInterface() {propertyValues {Curve {
        defaultValue=1
        keys [
         {arriveTangent=9.0725, interpMode=RCIM_Cubic, leaveTangent=9.0725, tangentMode=RCTM_User, time=0, value=0}
         {arriveTangent=-1.5214, interpMode=RCIM_Cubic, leaveTangent=-1.5214, tangentMode=RCTM_User, time=0.0479, value=0.7927}
         {arriveTangent=-0.157, interpMode=RCIM_Cubic, leaveTangent=-0.157, tangentMode=RCTM_User, time=0.6015, value=0.3799}
         {arriveTangent=-0.0405, interpMode=RCIM_Cubic, leaveTangent=-0.0405, tangentMode=RCTM_User, time=1, value=0}
        ]
       }}})
    ScaleA=true
    ScaleRGB=false
   )
   &GravityForce(Gravity=&Multiply_VectorByFloat(
     Vector=<0 0 -25>
     Float=&FloatFromCurve(FloatCurve=DataInterface() {propertyValues {Curve {
         defaultValue=1
         keys [<at 0 0 RCIM_Cubic>, {arriveTangent=2.0083, interpMode=RCIM_Cubic, leaveTangent=2.0083, tangentMode=RCTM_User, time=1, value=1}]
        }}})
    ))
   &Drag(Use Linear Drag=true, Use Rotational Drag=true)
   &SolveForcesAndVelocity(Clamp Velocity=false, Limit Acceleration=false)
  ]
  rendererClasses [&NiagaraSpriteRendererProperties]
  renderers [{rendererClass=&NiagaraSpriteRendererProperties}]
 }}
summary {
 systemName=NS_PaintCan_Lens
 emitters [{emitterName=Splash001, rendererClasses [&NiagaraSpriteRendererProperties]}]
 userVariables [User.FX_Scale_Global {type=NiagaraFloat, defaultValue=0, description=""}]
}
