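# Niagara system for a directional piano destruction effect in the grass easter egg, built entirely from mesh particles.
# A brief expanding glow accompanies GPU debris and black-or-white keys, partly launched in a tilted cone, plus simpler CPU bits without collision.
# GPU debris and keys enable collision by particle age, align to surfaces and shrink out.
# GPU leaves, tinted by User.LeafBaseColor, drift with gravity, wind and aerodynamic drag.
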NiagaraSystem /Hera/Environment/Blueprints/GrassEasteregg/FX/NS_Piano_Destruction_Directional
dependencies {
 usedDataInterfaces [
  &NiagaraDataInterfaceCurve
  &NiagaraDataInterfaceMeshRendererInfo
  &NiagaraDataInterfaceArrayFloat
  &NiagaraDataInterfaceColorCurve
  &NiagaraDataInterfaceCollisionQuery
  /Script/ChaosNiagara.NiagaraDataInterfacePhysicsField
  &NiagaraDataInterfaceLandscape
 ]
 usedDynamicInputs []
 usedModules [&SystemState]
}
emitters {
 Init_Glow001 {
  emitterUpdateScript [
   &EmitterState()
   &SpawnBurst_Instantaneous(Spawn Count=5, Use Spawn Probability=false)
  ]
  particleSpawnScript [
   &InitializeParticle(
    Color=<rgba 0.125 0.0731 0.0079 2>
    Lifetime=0.35
    Mesh Scale=<0.5 0.5 1>
    Mesh Scale Mode=::NewEnumerator4
    UsePositionOffset=false
    Write Material Random=true
   ) {module=InitializeParticle001}
   &InitialMeshOrientation(
    Rotation=&RandomRangeVector(Minimum=<0 0 0>, Maximum=<1 1 1>)
    Use Orientation Vector=true
    Use Rotation Vector=true
   ) {module=InitialMeshOrientation001}
   &SystemLocation(MaskSpawnGroup=false, Offset=&RandomRangeVector(Minimum=<0 0 100>, Maximum=<1 1 -100>))
  ]
  particleUpdateScript [
   &ParticleState()
   &DynamicMaterialParameters(
    Index 0 Param 1=&FloatFromCurve(FloatCurve=DataInterface() {propertyValues {Curve {keys [<at 0 1>,<at 1 1>]}}})
    Index 0 Param 2=1
    Index 0 Param 3=1
    Index 0 Param 4=1
    Module.Param0WriteEnabled=true
    Module.Param1WriteEnabled=true
    Module.Param2WriteEnabled=true
    Module.Param3WriteEnabled=true
   )
   &ScaleColor(
    Scale Alpha=&FloatFromCurve(FloatCurve=DataInterface() {propertyValues {Curve {keys [<at 0 0.5 RCIM_Cubic>,<at 1 0>]}}})
    Scale RGB=&MakeVectorFromLinearColorRGB(Color=&ColorFromCurve(ColorCurve=DataInterface() {propertyValues {
        AlphaCurve {keys [<at 0 1 RCIM_Cubic>]}
        BlueCurve {keys [<at 0 1 RCIM_Cubic>,<at 1 1 RCIM_Cubic>]}
        GreenCurve {keys [<at 0 1 RCIM_Cubic>,<at 1 0 RCIM_Cubic>]}
        RedCurve {keys [<at 0 1 RCIM_Cubic>,<at 1 0.2703 RCIM_Cubic>]}
       }}))
    ScaleA=true
    ScaleRGB=true
   )
   &ScaleMeshSize(Scale Factor=&VectorFromFloat(Value=&FloatFromCurve(FloatCurve=DataInterface() {propertyValues {Curve {keys [{interpMode=RCIM_Cubic, leaveTangent=1.6767, tangentMode=RCTM_Break, time=0, value=0}, <at 1 2>]}}})))
   &CameraOffset()
  ]
  rendererClasses [&NiagaraMeshRendererProperties]
  renderers [{rendererClass=&NiagaraMeshRendererProperties, rendererIndex=0}]
 }
 Leaves {
  simTarget=GPUComputeSim
  emitterUpdateScript [
   &EmitterState()
   &SpawnRate(SpawnRate=5000, Use Spawn Probability=false)
   &SpawnBurst_Instantaneous(Spawn Count=&MakeIntFromFloat(Float=100), Use Spawn Probability=false, UseLoopCountLimit=false)
  ]
  particleSpawnScript [
   &InitializeParticle(
    Color=$User.LeafBaseColor
    Lifetime Max=3.5
    Lifetime Min=1.5
    Lifetime Mode=::NewEnumerator1
    Mesh Renderer Array Visibility Mode=::NewEnumerator3
    Mesh Renderer Info=DataInterface() {propertyValues {MeshRenderer=&self:Leaves_1.NiagaraMeshRendererProperties_0}}
    Mesh Scale Max=<8 15 1>
    Mesh Scale Min=<4 10 1>
    Mesh Scale Mode=::NewEnumerator2
    UsePositionOffset=false
    Write Material Random=true
   )
   &InitialMeshOrientation(Use Rotation Vector=false)
   &ShapeLocation(
    Box Size=<200 200 10>
    MaskSpawnGroup=false
    Non Uniform Scale=<1 1 0>
    Override Seed=false
    Shape Origin=&AddVectorToPosition(Position=$StackContext.Position, Vector=<0 0 0>)
    Shape Primitive=::NewEnumerator2
    Surface Only=true
    UseOwnerScale=true
   )
   &AddVelocity(
    Module.Use Falloff Distance=false
    Origin Offset=<0 0 -500>
    Override Seed=false
    Velocity Mode=::NewEnumerator1
    Velocity Speed=&Multiply_Float(A=&RandomRangeFloat(Minimum=500, Maximum=1500), B=$Particles.Mass)
   )
   &ApplyInitialForces()
  ]
  particleUpdateScript [
   &ParticleState()
   &ScaleMeshSize(Scale Factor=&VectorFromFloat(Value=&FloatFromCurve(FloatCurve=DataInterface() {propertyValues {Curve {keys [<at 0.8 1 RCIM_Cubic>,<at 1 0 RCIM_Cubic>]}}}, Scale Curve=0.5)))
   &GravityForce()
   &WindForce(Ground Mask Elevation Mode:ENiagaraWindGroundMaskMode=::NewEnumerator1, Use Ground Mask=true, Use Physics Field=true, Wind Speed=<0 250 100>)
   &AerodynamicDrag(
    Aerodynamic Drag=0.02
    Aerodynamic Rotational Drag=0.02
    Drag Shape Curve=DataInterface() {propertyValues {Curve {keys [
        {interpMode=RCIM_Cubic, tangentMode=RCTM_User, time=0, value=0.1}
        {arriveTangent=2.0799, interpMode=RCIM_Cubic, leaveTangent=2.0799, tangentMode=RCTM_User, time=1, value=1}
       ]}}}
    Lift Shape Curve=DataInterface() {propertyValues {Curve {keys [
        {interpMode=RCIM_Cubic, tangentMode=RCTM_User, time=0, value=0}
        {interpMode=RCIM_Cubic, tangentMode=RCTM_User, time=0.5, value=1}
        {tangentMode=RCTM_User, time=1, value=0}
       ]}}}
    Rotation Strength:NiagaraFloat=&RandomRangeFloat(Minimum=3, Maximum=8)
    Use Linear Drag=true
    Use Rotational Drag=true
   )
   &SolveForcesAndVelocity(Clamp Velocity=false, Limit Acceleration=false)
   &DynamicMaterialParameters(
    Index 0 Param 1=&MakeCustomFloatFromBool(
     Bool=false
     True Float=&FloatFromCurve(FloatCurve=DataInterface() {propertyValues {Curve {keys [
          {arriveTangent=-9.3075, interpMode=RCIM_Cubic, leaveTangent=-9.3075, tangentMode=RCTM_User, time=0, value=2}
          {arriveTangent=-2.0574, interpMode=RCIM_Cubic, leaveTangent=-2.0574, tangentMode=RCTM_User, time=0.2, value=0.625}
          <at 0.75 0 RCIM_Cubic>
         ]}}})
     False Float=0
    )
    Index 0 Param 2=&MakeCustomFloatFromBool(Bool=false, True Float=1, False Float=0)
    Module.Param0WriteEnabled=true
    Module.Param1WriteEnabled=true
    Module.Param2WriteEnabled=false
    Module.Param3WriteEnabled=false
   )
  ]
  rendererClasses [&NiagaraMeshRendererProperties]
  renderers [{rendererClass=&NiagaraMeshRendererProperties, rendererIndex=0}]
 }
 Low_PianoBits {
  emitterUpdateScript [
   &EmitterState()
   &SpawnBurst_Instantaneous(Spawn Count=35, Use Spawn Probability=false, UseLoopCountLimit=false)
  ]
  particleSpawnScript [
   &FN_SetSpawnGroup(Number of Spawn Groups=2)
   /SetVariables(Particles.RandomFloat=&RandomRangeFloat(Minimum=0, Maximum=1)) {isSetParameters=true, module=SetVariables_#1}
   &InitializeParticle(
    Color=<rgba 0.854 1.1875 1.25 1>
    Lifetime=&RandomRangeFloat(Minimum=4, Maximum=4)
    Mass=/Niagara/DynamicInputs/Math/CalculateParticleRadius(
     Radius Calculation Type:ENiagaraCollisionRadiusOptions=::NewEnumerator1
     Particle Radius:NiagaraFloat=1
     Mesh Dimensions:Vector3f=<2 2 2>
     Method for Calculating Particle Radius:ENiagaraCalculateRadiusOptions=::NewEnumerator2
    )
    Mass Mode=::NewEnumerator1
    Mesh Scale=&Multiply_VectorByFloat(
     Vector=&RandomRangeVector(Minimum=<2 2 2>, Maximum=<5 5 5>)
     Float=&FloatFromCurve(
      FloatCurve=DataInterface() {propertyValues {Curve {keys [{arriveTangent=-1.7728, interpMode=RCIM_Cubic, leaveTangent=-1.7728, tangentMode=RCTM_User, time=0, value=1}, <at 1 0.2 RCIM_Cubic>]}}}
      CurveIndex=&Add_Float(A=$Particles.RandomFloat, B=&RandomRangeFloat(Minimum=-0.2, Maximum=0.2))
      Scale Curve=0.25
     )
    )
    Mesh Scale Mode=::NewEnumerator4
    UsePositionOffset=false
    Write Material Random=true
   )
   &ShapeLocation(
    Box Size=<200 200 10>
    MaskSpawnGroup=false
    Non Uniform Scale=<1 1 0>
    Override Seed=false
    Shape Origin=&AddVectorToPosition(Position=$StackContext.Position, Vector=<0 0 0>)
    Shape Primitive=::NewEnumerator2
    Surface Only=true
    UseOwnerScale=true
   )
   &InitialMeshOrientation(Use Rotation Vector=false)
   &AddVelocity(
    Module.Use Falloff Distance=false
    Origin Offset=<0 0 -500>
    Override Seed=false
    Velocity Mode=::NewEnumerator1
    Velocity Speed=&Multiply_Float(A=&RandomRangeFloat(Minimum=500, Maximum=1500), B=$Particles.Mass)
   )
   &MeshRotationForce(Rotation=&RandomVector(Vector Scale=&Multiply_Float(A=25, B=$Particles.Mass)))
   &PointForce(
    Force Origin=&AddVectorToPosition(Position=/Niagara/DynamicInputs/Helpers/SimulationPosition(), Vector=<0 0 0>)
    Force Strength=25
    Module.Use Falloff=false
    Module.UseLinearForce=false
   )
   &ApplyInitialForces()
   /SetVariables(Particles.VisibilityTag=&RandomRangeInt(Minimum=0, Maximum=3)) {isSetParameters=true, module=SetVariables_#2}
  ]
  particleUpdateScript [
   &ParticleState()
   &ScaleMeshSize(Scale Factor=&VectorFromFloat(Value=&FloatFromCurve(FloatCurve=DataInterface() {propertyValues {Curve {keys [<at 0 1 RCIM_Cubic>, {arriveTangent=-1, interpMode=RCIM_Cubic, leaveTangent=-1, time=0.7, value=1}, <at 1 0 RCIM_Cubic>]}}})))
   &Drag(
    Drag=&FloatFromCurve(
     FloatCurve=DataInterface() {propertyValues {Curve {keys [
         {arriveTangent=13.4118, interpMode=RCIM_Cubic, leaveTangent=13.4118, leaveTangentWeight=0.4483, tangentMode=RCTM_User, time=0, value=1}
         {arriveTangent=5, arriveTangentWeight=0.17, interpMode=RCIM_Constant, leaveTangentWeight=1.5297, time=0.1, value=0}
         {arriveTangentWeight=0.3, interpMode=RCIM_Cubic, tangentMode=RCTM_User, time=1, value=0}
        ]}}}
     Scale Curve=2
    )
    Use Linear Drag=true
    Use Rotational Drag=true
   )
   &GravityForce(Gravity=&RandomRangeVector(Minimum=<0 0 -980>, Maximum=<0 0 -1600>))
   &CurlNoiseForce(Bias Noise=false, Pan Noise=false)
   &SolveForcesAndVelocity(Clamp Velocity=false, Limit Acceleration=false)
   &CameraOffset(Camera Offset Amount=-50)
  ]
  rendererClasses [&NiagaraMeshRendererProperties]
  renderers [
   {rendererClass=&NiagaraMeshRendererProperties, rendererIndex=0}
   {rendererClass=&NiagaraMeshRendererProperties, rendererIndex=1}
   {rendererClass=&NiagaraMeshRendererProperties, rendererIndex=2}
   {rendererClass=&NiagaraMeshRendererProperties, rendererIndex=3}
  ]
 }
 PianoBits {
  simTarget=GPUComputeSim
  emitterUpdateScript [
   &EmitterState()
   &SpawnBurst_Instantaneous(Spawn Count=150, Use Spawn Probability=false, UseLoopCountLimit=false)
   &SpawnRate(SpawnRate=2500, Use Spawn Probability=false)
   &SpawnBurst_Instantaneous(Spawn Count=150, Spawn Time=1.65, Use Spawn Probability=false, UseLoopCountLimit=false) {module=SpawnBurst_Instantaneous001}
  ]
  particleSpawnScript [
   /SetVariables(Particles.RandomFloat=&RandomRangeFloat(Minimum=0, Maximum=1)) {isSetParameters=true, module=SetVariables_#1}
   &InitializeParticle(
    Color=<rgba 0.854 1.1875 1.25 1>
    Lifetime=&RandomRangeFloat(Minimum=10, Maximum=10)
    Mass=&CalculateParticleRadius(
     Radius Calculation Type:ENiagaraCollisionRadiusOptions=::NewEnumerator1
     Particle Radius:NiagaraFloat=1
     Mesh Dimensions:Vector3f=<2 2 2>
     Method for Calculating Particle Radius:ENiagaraCalculateRadiusOptions=::NewEnumerator2
    )
    Mass Mode=::NewEnumerator1
    Mesh Renderer Array Visibility Mode=::NewEnumerator3
    Mesh Renderer Info=DataInterface() {propertyValues {MeshRenderer=&self:PianoBits_0.NiagaraMeshRendererProperties_0}}
    Mesh Scale=&Multiply_VectorByFloat(
     Vector=&RandomRangeVector(Minimum=<2 2 2>, Maximum=<5 5 5>)
     Float=&FloatFromCurve(
      FloatCurve=DataInterface() {propertyValues {Curve {keys [{arriveTangent=-1.7728, interpMode=RCIM_Cubic, leaveTangent=-1.7728, tangentMode=RCTM_User, time=0, value=1}, <at 1 0.2 RCIM_Cubic>]}}}
      CurveIndex=&Add_Float(A=$Particles.RandomFloat, B=&RandomRangeFloat(Minimum=-0.2, Maximum=0.2))
      Scale Curve=0.25
     )
    )
    Mesh Scale Mode=::NewEnumerator4
    UsePositionOffset=false
    Write Material Random=true
   )
   &ShapeLocation(
    Box Size=<200 200 10>
    MaskSpawnGroup=false
    Non Uniform Scale=<1 1 0>
    Override Seed=false
    Shape Origin=&AddVectorToPosition(Position=$StackContext.Position, Vector=<0 0 0>)
    Shape Primitive=::NewEnumerator2
    Surface Only=true
    UseOwnerScale=true
   )
   &InitialMeshOrientation(Use Rotation Vector=false)
   &AddVelocity(
    Module.Use Falloff Distance=false
    Origin Offset=<0 200 -500>
    Override Seed=false
    Velocity Mode=::NewEnumerator1
    Velocity Speed=&Multiply_Float(A=&RandomRangeFloat(Minimum=500, Maximum=1500), B=$Particles.Mass)
   )
   &MeshRotationForce(Rotation=&RandomVector(Vector Scale=&Multiply_Float(A=5, B=$Particles.Mass)))
   &PointForce(
    Force Origin=&AddVectorToPosition(Position=&SimulationPosition(), Vector=<0 0 0>)
    Force Strength=25
    Module.Use Falloff=false
    Module.UseLinearForce=false
   )
   &ApplyInitialForces()
   /SetVariables(Particles.VisibilityTag=&RandomRangeInt(Minimum=0, Maximum=3)) {isSetParameters=true, module=SetVariables_#2}
  ]
  particleUpdateScript [
   &ParticleState()
   &ScaleMeshSize(Scale Factor=&VectorFromFloat(Value=&FloatFromCurve(FloatCurve=DataInterface() {propertyValues {Curve {keys [<at 0 1 RCIM_Cubic>, {arriveTangent=-1, interpMode=RCIM_Cubic, leaveTangent=-1, time=0.7, value=1}, <at 1 0 RCIM_Cubic>]}}})))
   &Drag(
    Drag=&FloatFromCurve(FloatCurve=DataInterface() {propertyValues {Curve {keys [
         {arriveTangent=13.4118, interpMode=RCIM_Cubic, leaveTangent=13.4118, leaveTangentWeight=0.4483, tangentMode=RCTM_User, time=0, value=1}
         {arriveTangent=5, arriveTangentWeight=0.17, interpMode=RCIM_Constant, leaveTangentWeight=1.5297, time=0.1, value=0}
         {arriveTangentWeight=0.3, interpMode=RCIM_Cubic, tangentMode=RCTM_User, time=1, value=0}
        ]}}})
    Use Linear Drag=true
    Use Rotational Drag=true
   )
   &GravityForce(Gravity=&RandomRangeVector(Minimum=<0 0 -980>, Maximum=<0 0 -1600>))
   &CurlNoiseForce(Bias Noise=false, Pan Noise=false)
   &Collision(
    Age Colliding Particles=false
    Collision Enabled=&SetBoolByFloatComparison(A=$Particles.Age, B=0.5, Comparison Type=::NewEnumerator2)
    Control Roll On Collision=true
    Friction=0
    GPU Collision Type=::NewEnumerator2
    KillOnCollision=false
    Mesh Dimensions:Vector3f=<5 5 5>
    Module.Control Rest State Via Surface Normal=false
    Module.EnableMaxCollisionCount=false
    Module.Randomize Collision Normal Vector=true
    Particle Radius Scale=5
    Radius Calculation Type=::NewEnumerator1
    Radius Scale for Rotation=250
    Randomize Collision Normal=0
    Restitution=0.2
    ScaleRateOfRotation=true
   )
   /Niagara/Modules/Collision/AlignParticlesWithCollisionPlane(Allow Alignment on X=false, Allow Alignment on Y=false, Calming Rate When Bouncing=1)
   &SolveForcesAndVelocity(Clamp Velocity=false, Limit Acceleration=false)
   &UpdateMeshOrientation(
    Mesh Renderer=DataInterface() {propertyValues {MeshRenderer=&self:PianoBits_0.NiagaraMeshRendererProperties_0}}
    Orientation Method=::NewEnumerator6
    Rotation Rate=2
   )
   &CameraOffset()
  ]
  rendererClasses [&NiagaraMeshRendererProperties]
  renderers [{rendererClass=&NiagaraMeshRendererProperties, rendererIndex=0}]
 }
 PianoBits001 {
  simTarget=GPUComputeSim
  emitterUpdateScript [
   &EmitterState()
   &SpawnRate(SpawnRate=5000, Use Spawn Probability=false)
   &SpawnBurst_Instantaneous(Spawn Count=250, Use Spawn Probability=false, UseLoopCountLimit=false)
  ]
  particleSpawnScript [
   &FN_SetSpawnGroup(Number of Spawn Groups=2)
   /SetVariables(Particles.RandomFloat=&RandomRangeFloat(Minimum=0, Maximum=1)) {isSetParameters=true, module=SetVariables_#1}
   &InitializeParticle(
    Color=<rgba 0.854 1.1875 1.25 1>
    Lifetime=&RandomRangeFloat(Minimum=8, Maximum=8)
    Mass=&CalculateParticleRadius(
     Radius Calculation Type:ENiagaraCollisionRadiusOptions=::NewEnumerator1
     Particle Radius:NiagaraFloat=1
     Mesh Dimensions:Vector3f=<2 2 2>
     Method for Calculating Particle Radius:ENiagaraCalculateRadiusOptions=::NewEnumerator2
    )
    Mass Mode=::NewEnumerator1
    Mesh Renderer Array Visibility Mode=::NewEnumerator3
    Mesh Renderer Info=DataInterface() {propertyValues {MeshRenderer=&self:PianoBits001_1.NiagaraMeshRendererProperties_0}}
    Mesh Scale=&Multiply_VectorByFloat(
     Vector=&RandomRangeVector(Minimum=<1 0.2 0.2>, Maximum=<2 1 1>)
     Float=&FloatFromCurve(
      FloatCurve=DataInterface() {propertyValues {Curve {keys [{arriveTangent=-1.7728, interpMode=RCIM_Cubic, leaveTangent=-1.7728, tangentMode=RCTM_User, time=0, value=1}, <at 1 0.2 RCIM_Cubic>]}}}
      CurveIndex=&Add_Float(A=$Particles.RandomFloat, B=&RandomRangeFloat(Minimum=-0.2, Maximum=0.2))
      Scale Curve=0.25
     )
    )
    Mesh Scale Mode=::NewEnumerator4
    UsePositionOffset=false
    Write Material Random=true
   )
   &ShapeLocation(
    Box Size=<200 200 10>
    MaskSpawnGroup=false
    Non Uniform Scale=<1 1 0>
    Override Seed=false
    Shape Origin=&AddVectorToPosition(Position=$StackContext.Position, Vector=<0 0 0>)
    Shape Primitive=::NewEnumerator2
    Surface Only=true
    UseOwnerScale=true
   )
   &InitialMeshOrientation(Use Rotation Vector=false)
   &AddVelocity(
    Cone Axis=<0 -1 1>
    Inner Cone Angle=20
    Module.UseConeFalloff=false
    Module.UseVelDistribution=false
    Override Seed=false
    Velocity Mode=::NewEnumerator2
    Velocity Speed=&Multiply_Float(A=&RandomRangeFloat(Minimum=500, Maximum=2500), B=$Particles.Mass)
   )
   &MeshRotationForce(Rotation=&RandomVector(Vector Scale=&Multiply_Float(A=50, B=$Particles.Mass))) {enabled=false}
   &PointForce(
    Force Origin=&AddVectorToPosition(Position=&SimulationPosition(), Vector=<0 0 0>)
    Force Strength=50
    Module.Use Falloff=false
    Module.UseLinearForce=false
   )
   &ApplyInitialForces()
   /SetVariables(Particles.VisibilityTag=&RandomRangeInt(Minimum=0, Maximum=1)) {isSetParameters=true, module=SetVariables_#2}
  ]
  particleUpdateScript [
   &ParticleState()
   &ScaleMeshSize(Scale Factor=&VectorFromFloat(Value=&FloatFromCurve(FloatCurve=DataInterface() {propertyValues {Curve {keys [<at 0 1 RCIM_Cubic>, {arriveTangent=-1, interpMode=RCIM_Cubic, leaveTangent=-1, time=0.7, value=1}, <at 1 0 RCIM_Cubic>]}}})))
   &Drag(
    Drag=&FloatFromCurve(FloatCurve=DataInterface() {propertyValues {Curve {keys [
         {arriveTangent=13.4118, interpMode=RCIM_Cubic, leaveTangent=13.4118, leaveTangentWeight=0.4483, tangentMode=RCTM_User, time=0, value=1}
         {arriveTangent=5, arriveTangentWeight=0.17, interpMode=RCIM_Cubic, leaveTangentWeight=1.5297, tangentMode=RCTM_User, time=0.1, value=0}
         {interpMode=RCIM_Cubic, tangentMode=RCTM_User, time=0.273, value=0.9802}
         {arriveTangentWeight=0.3, interpMode=RCIM_Cubic, tangentMode=RCTM_User, time=1, value=0}
        ]}}})
    Use Linear Drag=true
    Use Rotational Drag=true
   )
   &GravityForce(Gravity=&RandomRangeVector(Minimum=<0 0 -980>, Maximum=<0 0 -1600>))
   &CurlNoiseForce(Bias Noise=false, Pan Noise=false) {enabled=false}
   &Collision(
    Age Colliding Particles=false
    Collision Enabled=&SetBoolByFloatComparison(A=$Particles.Age, B=0.5, Comparison Type=::NewEnumerator2)
    Control Roll On Collision=true
    Enable Rest State=false
    Friction=0.2
    GPU Collision Type=::NewEnumerator2
    KillOnCollision=false
    Mesh Dimensions:Vector3f=<50 50 50>
    Module.EnableMaxCollisionCount=false
    Module.Randomize Collision Normal Vector=true
    Particle Radius Scale=2
    Radius Calculation Type=::NewEnumerator1
    Radius Scale for Rotation=25
    Randomize Collision Normal=0.2
    Restitution=0.1
    ScaleRateOfRotation=true
   )
   &AlignParticlesWithCollisionPlane(Allow Alignment on X=false, Allow Alignment on Y=false, Calming Rate When Resting=25, Rotational Dampening Power=25)
   &MeshRotationForce(Rotation=&MakeCustomVectorFromBool(Bool=$StackContext.HasCollided, True Vector=<0 0 0>, False Vector=<5 5 5>)) {module=MeshRotationForce001}
   &SolveForcesAndVelocity(
    Clamp Velocity=true
    Limit Acceleration=false
    Speed Limit:NiagaraFloat=&Multiply_Float(
     A=1500
     B=&Lerp_Float(A=0, B=1, Alpha=&MakeCustomFloatFromBool(Bool=$StackContext.HasCollided, True Float=0, False Float=1))
    )
   ) {enabled=false}
   &UpdateMeshOrientation(
    Mesh Renderer=DataInterface() {propertyValues {MeshRenderer=&self:PianoBits001_1.NiagaraMeshRendererProperties_0}}
    Orientation Method=::NewEnumerator6
    Particle Radius Scale=3
   )
   &SolveForcesAndVelocity(Clamp Velocity=false, Limit Acceleration=false) {module=SolveForcesAndVelocity001}
   &CameraOffset()
  ]
  rendererClasses [&NiagaraMeshRendererProperties]
  renderers [{rendererClass=&NiagaraMeshRendererProperties, rendererIndex=0}]
 }
 PianoKeys {
  simTarget=GPUComputeSim
  emitterUpdateScript [
   &EmitterState()
   &SpawnBurst_Instantaneous(Spawn Count=200, Use Spawn Probability=false, UseLoopCountLimit=false)
  ]
  particleSpawnScript [
   /SetVariables(Particles.RandomFloat=&RandomRangeFloat(Minimum=0, Maximum=1)) {isSetParameters=true, module=SetVariables_#1}
   &InitializeParticle(
    Color=&LerpLinearColors(
     StartColor=<rgba 1 1 1 1>
     EndColor=<rgba 0 0 0 1>
     LerpFactor=&SelectFloatFromArray(
      Array Sampling Mode=::NewEnumerator0
      Direct Array Index=?
      Float Selection Array=DataInterface() {propertyValues {FloatData [0,1]}}
      Random Seed=0
     )
    )
    Lifetime=&RandomRangeFloat(Minimum=3, Maximum=6)
    Mesh Scale Max=<25 5 5>
    Mesh Scale Min=<25 5 5>
    Mesh Scale Mode=::NewEnumerator2
    UsePositionOffset=false
    Write Material Random=true
   )
   &ShapeLocation(
    Box Size=<150 200 5>
    MaskSpawnGroup=false
    Non Uniform Scale=<1 1 0>
    Override Seed=false
    Shape Origin=&AddVectorToPosition(Position=$StackContext.Position, Vector=<0 0 0>)
    Shape Primitive=::NewEnumerator2
    Surface Only=false
    UseOwnerScale=true
   )
   &InitialMeshOrientation(Use Rotation Vector=false)
   &AddVelocity(
    Cone Axis=<0 -1 1>
    Inner Cone Angle=20
    Module.UseConeFalloff=false
    Module.UseVelDistribution=false
    Override Seed=false
    Velocity Mode=::NewEnumerator2
    Velocity Speed=&Multiply_Float(A=&RandomRangeFloat(Minimum=500, Maximum=800), B=$Particles.Mass)
   )
   &MeshRotationForce(Rotation=&RandomVector(Vector Scale=25))
   &PointForce(
    Force Origin=&AddVectorToPosition(Position=&SimulationPosition(), Vector=<0 0 2062.8154>)
    Force Strength=250
    Module.Use Falloff=false
    Module.UseLinearForce=false
   )
   &ApplyInitialForces()
   &AddVelocity(Velocity=&RandomRangeVector(Minimum=<0 0 0>, Maximum=<0 0 2343.4893>)) {module=AddVelocity001}
   /SetVariables(Particles.VisibilityTag=&RandomRangeInt(Minimum=0, Maximum=1)) {isSetParameters=true, module=SetVariables_#2}
  ]
  particleUpdateScript [
   &ParticleState()
   &ScaleMeshSize(Scale Factor=&VectorFromFloat(Value=&FloatFromCurve(FloatCurve=DataInterface() {propertyValues {Curve {keys [<at 0 1 RCIM_Cubic>, {arriveTangent=-1, interpMode=RCIM_Cubic, leaveTangent=-1, time=0.7, value=1}, <at 1 0 RCIM_Cubic>]}}})))
   &Drag(
    Drag=&FloatFromCurve(
     FloatCurve=DataInterface() {propertyValues {Curve {keys [
         {arriveTangent=13.4118, interpMode=RCIM_Cubic, leaveTangent=13.4118, leaveTangentWeight=0.4483, tangentMode=RCTM_User, time=0, value=1}
         {arriveTangent=5, arriveTangentWeight=0.17, interpMode=RCIM_Constant, leaveTangentWeight=1.5297, time=0.1, value=0}
         {arriveTangentWeight=0.3, interpMode=RCIM_Cubic, tangentMode=RCTM_User, time=1, value=0}
        ]}}}
     Scale Curve=2
    )
    Use Linear Drag=true
    Use Rotational Drag=true
   )
   &GravityForce(Gravity=&RandomRangeVector(Minimum=<0 0 -1960.4098>, Maximum=<0 0 -1033.373>))
   &Collision(
    Age Colliding Particles=false
    Collision Enabled=&SetBoolByFloatComparison(A=$Particles.Age, B=0.25, Comparison Type=::NewEnumerator2)
    GPU Collision Type=::NewEnumerator2
    KillOnCollision=false
    Module.Control Rest State Via Surface Normal=false
    Module.EnableMaxCollisionCount=false
    Module.Randomize Collision Normal Vector=true
    Radius Calculation Type=::NewEnumerator1
    Restitution=0.25
   )
   &AlignParticlesWithCollisionPlane(Allow Alignment on X=false, Allow Alignment on Y=false, Calming Rate When Bouncing=1)
   &MeshRotationForce(Rotation=&MakeCustomVectorFromBool(Bool=$StackContext.HasCollided, True Vector=<0 0 0>, False Vector=<1 1 1>)) {module=MeshRotationForce001}
   &SolveForcesAndVelocity(Clamp Velocity=false, Limit Acceleration=false)
   &UpdateMeshOrientation(
    Mesh Renderer=DataInterface() {propertyValues {MeshRenderer=&self:PianoKeys_0.NiagaraMeshRendererProperties_0}}
    Orientation Method=::NewEnumerator6
    Particle Radius Scale=3
    Surface Normal:Vector3f=<1 0 0>
   )
   &CameraOffset(Camera Offset Amount=-50)
  ]
  rendererClasses [&NiagaraMeshRendererProperties]
  renderers [{rendererClass=&NiagaraMeshRendererProperties, rendererIndex=0}]
 }
}
summary {
 systemName=NS_Piano_Destruction_Directional
 emitters [
  {emitterName=PianoBits, simTarget=GPUComputeSim, rendererClasses [&NiagaraMeshRendererProperties]}
  {emitterName=PianoKeys, simTarget=GPUComputeSim, rendererClasses [&NiagaraMeshRendererProperties]}
  {emitterName=PianoBits001, simTarget=GPUComputeSim, rendererClasses [&NiagaraMeshRendererProperties]}
  {emitterName=Low_PianoBits, rendererClasses [&NiagaraMeshRendererProperties]}
  {emitterName=Init_Glow001, rendererClasses [&NiagaraMeshRendererProperties]}
  {emitterName=Leaves, simTarget=GPUComputeSim, rendererClasses [&NiagaraMeshRendererProperties]}
 ]
 userVariables [User.LeafBaseColor {type=LinearColor, description="", defaultValue {struct=LinearColor, value=<rgba 0.4896 0.3043 0.0076 1>}}]
}
